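# Animation Blueprint for the Cosmos backpack pet, animating it in reaction to its owning player and switching between front-end and in-game poses.
# The cached Default state machine covers locomotion, hunker, jump, cower and underwater, with nested start/loop machines and an additive water-entry layer.
# Transitions read the player's anim instance for movement, skydiving, melee and swimming, alongside local flags.
# Index variables select locomotion break, cower start and underwater exit variants.
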AnimBlueprint /BRCosmetics/Characters/Pets/Pet_M_MED_Cosmos/Meshes/M_MED_Cosmos_Pet_AnimBP
target_skeleton=/BRCosmetics/Characters/PetsNoCarrier/Pet_M_MED_Cosmos/Meshes/M_MED_Cosmos_Pet_Skeleton
graphs [
 AnimGraph {nodes [
   AnimGraphNode_StateMachine_2 = FrontEndSM() {is=title AnimGraphNode_StateMachine}
   Output Pose(Result=Blend Poses by bool(
     BlendPose_0=@AnimGraphNode_StateMachine_2
     BlendPose_1=Use cached pose 'CosmosPetAnims'()
     BlendTime_0=0
     BlendTime_1=0
     bActiveValue=Get bIsInFrontEnd()
    ) {blendType=Linear}) {layerGroup=None}
   AnimGraphNode_StateMachine_0 = Default() {is=title AnimGraphNode_StateMachine}
   Save cached pose 'CosmosPetAnims'(Pose=@AnimGraphNode_StateMachine_0) {cacheName=CosmosPetAnims}
   AnimGraphNode_StateMachine_1 = FrontEnd() {is=title AnimGraphNode_StateMachine}
   Slot 'DefaultSlot'(Source=Blend Poses by bool(BlendPose_0=@AnimGraphNode_StateMachine_1) {blendType=Linear})
  ]}
 Default {
  entry=Locomotion
  path=AnimGraph.AnimGraphNode_StateMachine_0.Default
  ids=id
  nodes [
   AnimStateNode_0 = Locomotion() {
    is=title AnimStateNode
    stateEntered {guid=#1}
    stateFullyBlended {guid=#2}
    stateLeft {guid=#3}
   }
   AnimStateNode_10 = CowerEnd() {is=title AnimStateNode}
   AnimStateNode_11 = UnderWaterStop() {is=title AnimStateNode}
   AnimStateNode_2 = Hunker() {is=title AnimStateNode}
   AnimStateNode_4 = HunkerStop() {is=title AnimStateNode}
   AnimStateNode_5 = UnderWater() {is=title AnimStateNode, stateEntered {guid=#4, notifyName=UnderWaterEntered}}
   AnimStateNode_6 = Jump() {is=title AnimStateNode}
   AnimStateNode_7 = JumpLand() {is=title AnimStateNode}
   AnimStateNode_8 = Cower() {is=title AnimStateNode}
   AnimStateTransitionNode_1 = AnimStateTransitionNode()
   AnimStateTransitionNode_10 = AnimStateTransitionNode()
   AnimStateTransitionNode_11 = AnimStateTransitionNode()
   AnimStateTransitionNode_12 = AnimStateTransitionNode()
   AnimStateTransitionNode_13 = AnimStateTransitionNode()
   AnimStateTransitionNode_14 = AnimStateTransitionNode()
   AnimStateTransitionNode_15 = AnimStateTransitionNode() {crossfadeDuration=0}
   AnimStateTransitionNode_16 = AnimStateTransitionNode() {crossfadeDuration=0}
   AnimStateTransitionNode_17 = AnimStateTransitionNode() {bAutomaticRuleBasedOnSequencePlayerInState=true, crossfadeDuration=0.7}
   AnimStateTransitionNode_18 = AnimStateTransitionNode()
   AnimStateTransitionNode_19 = AnimStateTransitionNode() {bAutomaticRuleBasedOnSequencePlayerInState=true}
   AnimStateTransitionNode_20 = AnimStateTransitionNode()
   AnimStateTransitionNode_21 = AnimStateTransitionNode()
   AnimStateTransitionNode_4 = AnimStateTransitionNode() {crossfadeDuration=0.05}
   AnimStateTransitionNode_7 = AnimStateTransitionNode() {bAutomaticRuleBasedOnSequencePlayerInState=true}
   AnimStateTransitionNode_8 = AnimStateTransitionNode()
   AnimStateTransitionNode_9 = AnimStateTransitionNode() {bAutomaticRuleBasedOnSequencePlayerInState=true, crossfadeDuration=0.5}
  ]
  links [
   {
    dst=CowerEnd
    rule=AnimGraph.AnimGraphNode_StateMachine_0.Default.AnimStateTransitionNode_20.Transition
    src=Cower
    via=AnimStateTransitionNode_20
   }
   {
    dst=Locomotion
    rule=AnimGraph.AnimGraphNode_StateMachine_0.Default.AnimStateTransitionNode_7.Transition
    src=CowerEnd
    via=AnimStateTransitionNode_7
   }
   {
    dst=HunkerStop
    rule=AnimGraph.AnimGraphNode_StateMachine_0.Default.AnimStateTransitionNode_8.Transition
    src=Hunker
    via=AnimStateTransitionNode_8
   }
   {
    dst=Locomotion
    rule=AnimGraph.AnimGraphNode_StateMachine_0.Default.AnimStateTransitionNode_9.Transition
    src=HunkerStop
    via=AnimStateTransitionNode_9
   }
   {
    dst=JumpLand
    rule=AnimGraph.AnimGraphNode_StateMachine_0.Default.AnimStateTransitionNode_10.Transition
    src=Jump
    via=AnimStateTransitionNode_10
   }
   {
    dst=JumpLand
    rule=AnimGraph.AnimGraphNode_StateMachine_0.Default.AnimStateTransitionNode_14.Transition
    src=Jump
    via=AnimStateTransitionNode_14
   }
   {
    dst=JumpLand
    rule=AnimGraph.AnimGraphNode_StateMachine_0.Default.AnimStateTransitionNode_16.Transition
    src=Jump
    via=AnimStateTransitionNode_16
   }
   {
    dst=Locomotion
    rule=AnimGraph.AnimGraphNode_StateMachine_0.Default.AnimStateTransitionNode_11.Transition
    src=JumpLand
    via=AnimStateTransitionNode_11
   }
   {
    dst=Locomotion
    rule=AnimGraph.AnimGraphNode_StateMachine_0.Default.AnimStateTransitionNode_13.Transition
    src=JumpLand
    via=AnimStateTransitionNode_13
   }
   {
    dst=Locomotion
    rule=AnimGraph.AnimGraphNode_StateMachine_0.Default.AnimStateTransitionNode_17.Transition
    src=JumpLand
    via=AnimStateTransitionNode_17
   }
   {
    dst=Cower
    rule=AnimGraph.AnimGraphNode_StateMachine_0.Default.AnimStateTransitionNode_4.Transition
    src=Locomotion
    via=AnimStateTransitionNode_4
   }
   {
    dst=Hunker
    rule=AnimGraph.AnimGraphNode_StateMachine_0.Default.AnimStateTransitionNode_1.Transition
    src=Locomotion
    via=AnimStateTransitionNode_1
   }
   {
    dst=Jump
    rule=AnimGraph.AnimGraphNode_StateMachine_0.Default.AnimStateTransitionNode_15.Transition
    src=Locomotion
    via=AnimStateTransitionNode_15
   }
   {
    dst=UnderWater
    rule=AnimGraph.AnimGraphNode_StateMachine_0.Default.AnimStateTransitionNode_12.Transition
    src=Locomotion
    via=AnimStateTransitionNode_12
   }
   {
    dst=UnderWaterStop
    rule=AnimGraph.AnimGraphNode_StateMachine_0.Default.AnimStateTransitionNode_18.Transition
    src=UnderWater
    via=AnimStateTransitionNode_18
   }
   {
    dst=Locomotion
    rule=AnimGraph.AnimGraphNode_StateMachine_0.Default.AnimStateTransitionNode_19.Transition
    src=UnderWaterStop
    via=AnimStateTransitionNode_19
   }
   {
    dst=UnderWater
    rule=AnimGraph.AnimGraphNode_StateMachine_0.Default.AnimStateTransitionNode_21.Transition
    src=UnderWaterStop
    via=AnimStateTransitionNode_21
   }
  ]
 }
 Locomotion {
  path=AnimGraph.AnimGraphNode_StateMachine_0.Default.AnimStateNode_0.Locomotion
  nodes [
   AnimGraphNode_StateMachine_1 = Locomotion() {is=title AnimGraphNode_StateMachine}
   Output Animation Pose(Result=@AnimGraphNode_StateMachine_1)
  ]
 }
 Locomotion {
  entry=Idle
  path=AnimGraph.AnimGraphNode_StateMachine_0.Default.AnimStateNode_0.Locomotion.AnimGraphNode_StateMachine_1.Locomotion
  ids=id
  nodes [
   AnimStateConduitNode_0 = Conduit()
   AnimStateConduitNode_1 = Conduit_1() {is=title AnimStateConduitNode}
   AnimStateConduitNode_2 = Conduit_2() {is=title AnimStateConduitNode}
   AnimStateNode_0 = RandomLocoBreak() {is=title AnimStateNode, stateEntered {guid=#5, notifyName=Entered_RandomLocoBreak}}
   AnimStateNode_1 = MoveStart() {is=title AnimStateNode}
   AnimStateNode_2 = MoveLoop() {is=title AnimStateNode}
   AnimStateNode_3 = Idle()
   AnimStateNode_4 = SkyDiveBreak() {is=title AnimStateNode, stateEntered {guid=#6, notifyName=Entered_SkyDiveBreak}}
   AnimStateNode_6 = MoveStop() {is=title AnimStateNode}
   AnimStateTransitionNode_0 = AnimStateTransitionNode()
   AnimStateTransitionNode_1 = AnimStateTransitionNode()
   AnimStateTransitionNode_10 = AnimStateTransitionNode()
   AnimStateTransitionNode_11 = AnimStateTransitionNode()
   AnimStateTransitionNode_12 = AnimStateTransitionNode() {bAutomaticRuleBasedOnSequencePlayerInState=true}
   AnimStateTransitionNode_13 = AnimStateTransitionNode()
   AnimStateTransitionNode_14 = AnimStateTransitionNode()
   AnimStateTransitionNode_15 = AnimStateTransitionNode()
   AnimStateTransitionNode_17 = AnimStateTransitionNode()
   AnimStateTransitionNode_19 = AnimStateTransitionNode()
   AnimStateTransitionNode_2 = AnimStateTransitionNode()
   AnimStateTransitionNode_21 = AnimStateTransitionNode() {bAutomaticRuleBasedOnSequencePlayerInState=true}
   AnimStateTransitionNode_3 = AnimStateTransitionNode() {bAutomaticRuleBasedOnSequencePlayerInState=true}
   AnimStateTransitionNode_4 = AnimStateTransitionNode()
   AnimStateTransitionNode_5 = AnimStateTransitionNode()
   AnimStateTransitionNode_7 = AnimStateTransitionNode() {bAutomaticRuleBasedOnSequencePlayerInState=true}
   AnimStateTransitionNode_9 = AnimStateTransitionNode()
  ]
  links [
   {
    dst=MoveLoop
    rule=AnimGraph.AnimGraphNode_StateMachine_0.Default.AnimStateNode_0.Locomotion.AnimGraphNode_StateMachine_1.Locomotion.AnimStateTransitionNode_13.Transition
    src=Conduit
    via=AnimStateTransitionNode_13
   }
   {
    dst=MoveStop
    rule=AnimGraph.AnimGraphNode_StateMachine_0.Default.AnimStateNode_0.Locomotion.AnimGraphNode_StateMachine_1.Locomotion.AnimStateTransitionNode_14.Transition
    src=Conduit
    via=AnimStateTransitionNode_14
   }
   {
    dst=RandomLocoBreak
    rule=AnimGraph.AnimGraphNode_StateMachine_0.Default.AnimStateNode_0.Locomotion.AnimGraphNode_StateMachine_1.Locomotion.AnimStateTransitionNode_1.Transition
    src=Conduit_1
    via=AnimStateTransitionNode_1
   }
   {
    dst=SkyDiveBreak
    rule=AnimGraph.AnimGraphNode_StateMachine_0.Default.AnimStateNode_0.Locomotion.AnimGraphNode_StateMachine_1.Locomotion.AnimStateTransitionNode_5.Transition
    src=Conduit_1
    via=AnimStateTransitionNode_5
   }
   {
    dst=MoveStart
    rule=AnimGraph.AnimGraphNode_StateMachine_0.Default.AnimStateNode_0.Locomotion.AnimGraphNode_StateMachine_1.Locomotion.AnimStateTransitionNode_15.Transition
    src=Conduit_2
    via=AnimStateTransitionNode_15
   }
   {
    dst=SkyDiveBreak
    rule=AnimGraph.AnimGraphNode_StateMachine_0.Default.AnimStateNode_0.Locomotion.AnimGraphNode_StateMachine_1.Locomotion.AnimStateTransitionNode_17.Transition
    src=Conduit_2
    via=AnimStateTransitionNode_17
   }
   {
    dst=Conduit_2
    rule=AnimGraph.AnimGraphNode_StateMachine_0.Default.AnimStateNode_0.Locomotion.AnimGraphNode_StateMachine_1.Locomotion.AnimStateTransitionNode_10.Transition
    src=Idle
    via=AnimStateTransitionNode_10
   }
   {
    dst=MoveStart
    rule=AnimGraph.AnimGraphNode_StateMachine_0.Default.AnimStateNode_0.Locomotion.AnimGraphNode_StateMachine_1.Locomotion.AnimStateTransitionNode_19.Transition
    src=Idle
    via=AnimStateTransitionNode_19
   }
   {
    dst=Conduit_1
    rule=AnimGraph.AnimGraphNode_StateMachine_0.Default.AnimStateNode_0.Locomotion.AnimGraphNode_StateMachine_1.Locomotion.AnimStateTransitionNode_0.Transition
    src=MoveLoop
    via=AnimStateTransitionNode_0
   }
   {
    dst=MoveStop
    rule=AnimGraph.AnimGraphNode_StateMachine_0.Default.AnimStateNode_0.Locomotion.AnimGraphNode_StateMachine_1.Locomotion.AnimStateTransitionNode_11.Transition
    src=MoveLoop
    via=AnimStateTransitionNode_11
   }
   {
    dst=MoveLoop
    rule=AnimGraph.AnimGraphNode_StateMachine_0.Default.AnimStateNode_0.Locomotion.AnimGraphNode_StateMachine_1.Locomotion.AnimStateTransitionNode_2.Transition
    src=MoveStart
    via=AnimStateTransitionNode_2
   }
   {
    dst=MoveLoop
    rule=AnimGraph.AnimGraphNode_StateMachine_0.Default.AnimStateNode_0.Locomotion.AnimGraphNode_StateMachine_1.Locomotion.AnimStateTransitionNode_21.Transition
    src=MoveStart
    via=AnimStateTransitionNode_21
   }
   {
    dst=Idle
    rule=AnimGraph.AnimGraphNode_StateMachine_0.Default.AnimStateNode_0.Locomotion.AnimGraphNode_StateMachine_1.Locomotion.AnimStateTransitionNode_3.Transition
    src=MoveStop
    via=AnimStateTransitionNode_3
   }
   {
    dst=Idle
    rule=AnimGraph.AnimGraphNode_StateMachine_0.Default.AnimStateNode_0.Locomotion.AnimGraphNode_StateMachine_1.Locomotion.AnimStateTransitionNode_4.Transition
    src=MoveStop
    via=AnimStateTransitionNode_4
   }
   {
    dst=Conduit
    rule=AnimGraph.AnimGraphNode_StateMachine_0.Default.AnimStateNode_0.Locomotion.AnimGraphNode_StateMachine_1.Locomotion.AnimStateTransitionNode_12.Transition
    src=RandomLocoBreak
    via=AnimStateTransitionNode_12
   }
   {
    dst=Idle
    rule=AnimGraph.AnimGraphNode_StateMachine_0.Default.AnimStateNode_0.Locomotion.AnimGraphNode_StateMachine_1.Locomotion.AnimStateTransitionNode_7.Transition
    src=SkyDiveBreak
    via=AnimStateTransitionNode_7
   }
   {
    dst=Idle
    rule=AnimGraph.AnimGraphNode_StateMachine_0.Default.AnimStateNode_0.Locomotion.AnimGraphNode_StateMachine_1.Locomotion.AnimStateTransitionNode_9.Transition
    src=SkyDiveBreak
    via=AnimStateTransitionNode_9
   }
  ]
 }
 Conduit {
  path=AnimGraph.AnimGraphNode_StateMachine_0.Default.AnimStateNode_0.Locomotion.AnimGraphNode_StateMachine_1.Locomotion.AnimStateConduitNode_0.Conduit
  nodes [Result(bCanEnterTransition=true)]
 }
 Conduit_1 {
  path=AnimGraph.AnimGraphNode_StateMachine_0.Default.AnimStateNode_0.Locomotion.AnimGraphNode_StateMachine_1.Locomotion.AnimStateConduitNode_1.Conduit_1
  nodes [Result(bCanEnterTransition=true)]
 }
 Conduit_2 {
  path=AnimGraph.AnimGraphNode_StateMachine_0.Default.AnimStateNode_0.Locomotion.AnimGraphNode_StateMachine_1.Locomotion.AnimStateConduitNode_2.Conduit_2
  nodes [Result(bCanEnterTransition=true)]
 }
 RandomLocoBreak {
  path=AnimGraph.AnimGraphNode_StateMachine_0.Default.AnimStateNode_0.Locomotion.AnimGraphNode_StateMachine_1.Locomotion.AnimStateNode_0.RandomLocoBreak
  nodes [
   n1 = Get FortPlayerAnimBP() -> (FortPlayerAnimBP:Fortnite M Avg Player Anim Blueprint Object Reference)
   Blend Poses by bool(
    BlendPose_0=Sequence Player 'Skydive_Break'!(bLoopAnimation:bool=false) -> (Pose) {sequence=/BRCosmetics/Animation/Game/Pets/Cosmos/Backpack/Locomotion/Skydive_Break}
    bActiveValue=Break Fort Anim Input Skydiving!(FortAnimInput_Skydiving:"Fort Anim Input Skydiving Structure (by ref)"=Get Skydiving!(self:Fort Player Anim Instance Object Reference=@n1) -> (Skydiving:Fort Anim Input Skydiving Structure)) -> (
     ParaGilderRootOffset:Vector=~
     ParaGliderRootRotationOverride:Rotator=~
     bUseParaGlideRootModifier:bool=~
     bIsSkydivingFromLaunchPad:bool=~
     bIsSkydivingFromBus:bool=~
     bIsInVortex:bool=~
     bIsDiving:bool=~
     bIsParachuteOpen:bool=~
     bIsSkydiving:bool=~
     bIsParachuteLeaning:bool=~
     bIsSkydiveLeaning:bool=~
     bIsLeaning:bool=~
     bIsSkydiveDiveMode:bool=~
     bPlaySkydiveDrift:bool=~
     bSkydiveDriftDelayActive:bool=~
     bSkydiveDriftAnimAllowed:bool=~
     bIsGliderRight:bool=~
     bIsGliderCenter:bool=~
     bIsGliderLeft:bool=~
     bIsGliderForward:bool=~
     bIsGliderBack:bool=~
     bHasDeployedAndIsStandingUpright:bool=~
     LocalAccelForward:float=~
     LocalAccelRight:float=~
     LocalVelocityRight:float=~
     SkydiveAimPitch:float=~
     SkydiveAimYaw:float=~
     DeployChuteAnimRate:float=~
     SkydiveDriftAnimRate:float=~
     SkydiveDriftAnimRateCurrent:float=~
     SkydiveFidgetAnimRate:float=~
     SkydiveFidgetAnimRateCurrent:float=~
     SkydiveAdditiveAlpha:float=~
     SkydiveDriftDelay:float=~
     SkydiveDriftAnim:int
     SkydiveDriftAnimMax:int
     LaunchpadAnim:int
     LocalAccelDir:ESkydivingDirection Enum
     DirectionLast:ESkydivingDirection Enum
     PoseAdditive:Anim Sequence Object Reference
    ) {is=title K2Node_BreakStruct}.bIsSkydiving
   ) {blendType=Linear}
   n2 = Blend Poses by int!(
    BlendPose_0:Pose Link Structure=Sequence Player 'Locomotion_Break1'!(bLoopAnimation:bool=false) -> (Pose) {sequence=/BRCosmetics/Animation/Game/Pets/Cosmos/Backpack/Locomotion/Locomotion_Break1}
    BlendPose_1:Pose Link Structure=Sequence Player 'Locomotion_Break2'!(bLoopAnimation:bool=false) -> (Pose) {sequence=/BRCosmetics/Animation/Game/Pets/Cosmos/Backpack/Locomotion/Locomotion_Break2}
    BlendPose_2:Pose Link Structure=Sequence Player 'Locomotion_Break3'!(bLoopAnimation:bool=false) -> (Pose) {sequence=/BRCosmetics/Animation/Game/Pets/Cosmos/Backpack/Locomotion/Locomotion_Break3}
    BlendTime_0:float=0.1
    BlendTime_1:float=0.1
    BlendTime_2:float=0.1
    ActiveChildIndex:int=Get Choose_LocoBreak_Index() -> (Choose_LocoBreak_Index:int=0)
   ) -> (Pose:Pose Link Structure) {
    is=title AnimGraphNode_BlendListByInt
    blendType=Linear
    blendPose [{linkId=-1, sourceLinkId=-1}, {linkId=-1, sourceLinkId=-1}, {linkId=-1, sourceLinkId=-1}]
    blendTime [0.1,0.1,0.1]
   }
   Output Animation Pose(Result=@n2)
  ]
 }
 MoveStart {
  path=AnimGraph.AnimGraphNode_StateMachine_0.Default.AnimStateNode_0.Locomotion.AnimGraphNode_StateMachine_1.Locomotion.AnimStateNode_1.MoveStart
  nodes [Output Animation Pose(Result=Sequence Player 'LocomotionStart'!(bLoopAnimation:bool=false) -> (Pose) {sequence=/BRCosmetics/Animation/Game/Pets/Cosmos/Backpack/Locomotion/LocomotionStart})]
 }
 MoveLoop {
  path=AnimGraph.AnimGraphNode_StateMachine_0.Default.AnimStateNode_0.Locomotion.AnimGraphNode_StateMachine_1.Locomotion.AnimStateNode_2.MoveLoop
  nodes [Output Animation Pose(Result=Sequence Player 'Locomotion'() {sequence=/BRCosmetics/Animation/Game/Pets/Cosmos/Backpack/Locomotion/Locomotion})]
 }
 Idle {
  path=AnimGraph.AnimGraphNode_StateMachine_0.Default.AnimStateNode_0.Locomotion.AnimGraphNode_StateMachine_1.Locomotion.AnimStateNode_3.Idle
  nodes [Output Animation Pose(Result=Sequence Player 'Idle'() {sequence=/BRCosmetics/Animation/Game/Pets/Cosmos/Backpack/Idle/Idle})]
 }
 SkyDiveBreak {
  path=AnimGraph.AnimGraphNode_StateMachine_0.Default.AnimStateNode_0.Locomotion.AnimGraphNode_StateMachine_1.Locomotion.AnimStateNode_4.SkyDiveBreak
  nodes [Output Animation Pose(Result=Sequence Player 'Skydive_Break'(bLoopAnimation=False) {bLoopAnimation=false, sequence=&Skydive_Break})]
 }
 MoveStop {
  path=AnimGraph.AnimGraphNode_StateMachine_0.Default.AnimStateNode_0.Locomotion.AnimGraphNode_StateMachine_1.Locomotion.AnimStateNode_6.MoveStop
  nodes [Output Animation Pose(Result=Sequence Player 'LocomotionStop'!(bLoopAnimation:bool=false) -> (Pose) {sequence=/BRCosmetics/Animation/Game/Pets/Cosmos/Backpack/Locomotion/LocomotionStop})]
 }
 Transition {
  path=AnimGraph.AnimGraphNode_StateMachine_0.Default.AnimStateNode_0.Locomotion.AnimGraphNode_StateMachine_1.Locomotion.AnimStateTransitionNode_0.Transition
  nodes [Result(bCanEnterTransition=Get AllowLocomotionBreak() -> (AllowLocomotionBreak:bool))]
 }
 Transition {
  path=AnimGraph.AnimGraphNode_StateMachine_0.Default.AnimStateNode_0.Locomotion.AnimGraphNode_StateMachine_1.Locomotion.AnimStateTransitionNode_1.Transition
  nodes [
   n1 = Get Skydiving(self=Get FortPlayerAnimBP())
   Result(bCanEnterTransition=NOT Boolean(A=Break Fort Anim Input Skydiving(FortAnimInput_Skydiving=@n1) -> (
      ParaGilderRootOffset=~
      ParaGliderRootRotationOverride=~
      bUseParaGlideRootModifier=~
      bIsSkydivingFromLaunchPad=~
      bIsSkydivingFromBus=~
      bIsInVortex=~
      bIsDiving=~
      bIsParachuteOpen=~
      bIsSkydiving=~
      bIsParachuteLeaning=~
      bIsSkydiveLeaning=~
      bIsLeaning=~
      bIsSkydiveDiveMode=~
      bPlaySkydiveDrift=~
      bSkydiveDriftDelayActive=~
      bSkydiveDriftAnimAllowed=~
      bIsGliderRight=~
      bIsGliderCenter=~
      bIsGliderLeft=~
      bIsGliderForward=~
      bIsGliderBack=~
      bHasDeployedAndIsStandingUpright=~
      LocalAccelForward=~
      LocalAccelRight=~
      LocalVelocityRight=~
      SkydiveAimPitch=~
      SkydiveAimYaw=~
      DeployChuteAnimRate=~
      SkydiveDriftAnimRate=~
      SkydiveDriftAnimRateCurrent=~
      SkydiveFidgetAnimRate=~
      SkydiveFidgetAnimRateCurrent=~
      SkydiveAdditiveAlpha=~
      SkydiveDriftDelay=~
     ).bIsSkydiving))
  ]
 }
 Transition {
  path=AnimGraph.AnimGraphNode_StateMachine_0.Default.AnimStateNode_0.Locomotion.AnimGraphNode_StateMachine_1.Locomotion.AnimStateTransitionNode_10.Transition
  nodes [Result(bCanEnterTransition=Get AllowLocomotionBreak())]
 }
 Transition {
  path=AnimGraph.AnimGraphNode_StateMachine_0.Default.AnimStateNode_0.Locomotion.AnimGraphNode_StateMachine_1.Locomotion.AnimStateTransitionNode_11.Transition
  nodes [
   n1 = Get FortPlayerAnimBP()
   Result(bCanEnterTransition=NOT Boolean(A=Get bIsMoving2D!(self:Fort Player Anim Instance Object Reference=@n1) -> (bIsMoving2D)))
  ]
 }
 Transition {
  path=AnimGraph.AnimGraphNode_StateMachine_0.Default.AnimStateNode_0.Locomotion.AnimGraphNode_StateMachine_1.Locomotion.AnimStateTransitionNode_12.Transition
  nodes [Result()]
 }
 Transition {
  path=AnimGraph.AnimGraphNode_StateMachine_0.Default.AnimStateNode_0.Locomotion.AnimGraphNode_StateMachine_1.Locomotion.AnimStateTransitionNode_13.Transition
  nodes [Result(bCanEnterTransition=Get bIsMoving2D(self=Get FortPlayerAnimBP()))]
 }
 Transition {
  path=AnimGraph.AnimGraphNode_StateMachine_0.Default.AnimStateNode_0.Locomotion.AnimGraphNode_StateMachine_1.Locomotion.AnimStateTransitionNode_14.Transition
  nodes [
   n1 = Get FortPlayerAnimBP()
   Result(bCanEnterTransition=NOT Boolean(A=Get bIsMoving2D(self=@n1)))
  ]
 }
 Transition {
  path=AnimGraph.AnimGraphNode_StateMachine_0.Default.AnimStateNode_0.Locomotion.AnimGraphNode_StateMachine_1.Locomotion.AnimStateTransitionNode_15.Transition
  nodes [
   n1 = Get Skydiving(self=Get FortPlayerAnimBP())
   Result(bCanEnterTransition=NOT Boolean(A=Break Fort Anim Input Skydiving(FortAnimInput_Skydiving=@n1) -> (
      ParaGilderRootOffset=~
      ParaGliderRootRotationOverride=~
      bUseParaGlideRootModifier=~
      bIsSkydivingFromLaunchPad=~
      bIsSkydivingFromBus=~
      bIsInVortex=~
      bIsDiving=~
      bIsParachuteOpen=~
      bIsSkydiving=~
      bIsParachuteLeaning=~
      bIsSkydiveLeaning=~
      bIsLeaning=~
      bIsSkydiveDiveMode=~
      bPlaySkydiveDrift=~
      bSkydiveDriftDelayActive=~
      bSkydiveDriftAnimAllowed=~
      bIsGliderRight=~
      bIsGliderCenter=~
      bIsGliderLeft=~
      bIsGliderForward=~
      bIsGliderBack=~
      bHasDeployedAndIsStandingUpright=~
      LocalAccelForward=~
      LocalAccelRight=~
      LocalVelocityRight=~
      SkydiveAimPitch=~
      SkydiveAimYaw=~
      DeployChuteAnimRate=~
      SkydiveDriftAnimRate=~
      SkydiveDriftAnimRateCurrent=~
      SkydiveFidgetAnimRate=~
      SkydiveFidgetAnimRateCurrent=~
      SkydiveAdditiveAlpha=~
      SkydiveDriftDelay=~
     ).bIsSkydiving))
  ]
 }
 Transition {
  path=AnimGraph.AnimGraphNode_StateMachine_0.Default.AnimStateNode_0.Locomotion.AnimGraphNode_StateMachine_1.Locomotion.AnimStateTransitionNode_17.Transition
  nodes [
   n1 = Get FortPlayerAnimBP()
   Result(bCanEnterTransition=Break Fort Anim Input Skydiving(FortAnimInput_Skydiving=Get Skydiving(self=@n1)) -> (
     ParaGilderRootOffset=~
     ParaGliderRootRotationOverride=~
     bUseParaGlideRootModifier=~
     bIsSkydivingFromLaunchPad=~
     bIsSkydivingFromBus=~
     bIsInVortex=~
     bIsDiving=~
     bIsParachuteOpen=~
     bIsSkydiving=~
     bIsParachuteLeaning=~
     bIsSkydiveLeaning=~
     bIsLeaning=~
     bIsSkydiveDiveMode=~
     bPlaySkydiveDrift=~
     bSkydiveDriftDelayActive=~
     bSkydiveDriftAnimAllowed=~
     bIsGliderRight=~
     bIsGliderCenter=~
     bIsGliderLeft=~
     bIsGliderForward=~
     bIsGliderBack=~
     bHasDeployedAndIsStandingUpright=~
     LocalAccelForward=~
     LocalAccelRight=~
     LocalVelocityRight=~
     SkydiveAimPitch=~
     SkydiveAimYaw=~
     DeployChuteAnimRate=~
     SkydiveDriftAnimRate=~
     SkydiveDriftAnimRateCurrent=~
     SkydiveFidgetAnimRate=~
     SkydiveFidgetAnimRateCurrent=~
     SkydiveAdditiveAlpha=~
     SkydiveDriftDelay=~
    ).bIsSkydiving)
  ]
 }
 Transition {
  path=AnimGraph.AnimGraphNode_StateMachine_0.Default.AnimStateNode_0.Locomotion.AnimGraphNode_StateMachine_1.Locomotion.AnimStateTransitionNode_19.Transition
  nodes [Result(bCanEnterTransition=Get bIsMoving2D(self=Get FortPlayerAnimBP()))]
 }
 Transition {
  path=AnimGraph.AnimGraphNode_StateMachine_0.Default.AnimStateNode_0.Locomotion.AnimGraphNode_StateMachine_1.Locomotion.AnimStateTransitionNode_2.Transition
  nodes [
   n1 = Get AllowLocomotionBreak()
   n2 = Get bIsMoving2D(self=Get FortPlayerAnimBP())
   Result(bCanEnterTransition=AND Boolean(A=NOT Boolean(A=@n2), B=NOT Boolean(A=@n1)))
  ]
 }
 Transition {
  path=AnimGraph.AnimGraphNode_StateMachine_0.Default.AnimStateNode_0.Locomotion.AnimGraphNode_StateMachine_1.Locomotion.AnimStateTransitionNode_21.Transition
  nodes [Result()]
 }
 Transition {
  path=AnimGraph.AnimGraphNode_StateMachine_0.Default.AnimStateNode_0.Locomotion.AnimGraphNode_StateMachine_1.Locomotion.AnimStateTransitionNode_3.Transition
  nodes [Result()]
 }
 Transition {
  path=AnimGraph.AnimGraphNode_StateMachine_0.Default.AnimStateNode_0.Locomotion.AnimGraphNode_StateMachine_1.Locomotion.AnimStateTransitionNode_4.Transition
  nodes [Result(bCanEnterTransition=Get bIsMoving2D(self=Get FortPlayerAnimBP()))]
 }
 Transition {
  path=AnimGraph.AnimGraphNode_StateMachine_0.Default.AnimStateNode_0.Locomotion.AnimGraphNode_StateMachine_1.Locomotion.AnimStateTransitionNode_5.Transition
  nodes [
   n1 = Get FortPlayerAnimBP()
   Result(bCanEnterTransition=Break Fort Anim Input Skydiving(FortAnimInput_Skydiving=Get Skydiving(self=@n1)) -> (
     ParaGilderRootOffset=~
     ParaGliderRootRotationOverride=~
     bUseParaGlideRootModifier=~
     bIsSkydivingFromLaunchPad=~
     bIsSkydivingFromBus=~
     bIsInVortex=~
     bIsDiving=~
     bIsParachuteOpen=~
     bIsSkydiving=~
     bIsParachuteLeaning=~
     bIsSkydiveLeaning=~
     bIsLeaning=~
     bIsSkydiveDiveMode=~
     bPlaySkydiveDrift=~
     bSkydiveDriftDelayActive=~
     bSkydiveDriftAnimAllowed=~
     bIsGliderRight=~
     bIsGliderCenter=~
     bIsGliderLeft=~
     bIsGliderForward=~
     bIsGliderBack=~
     bHasDeployedAndIsStandingUpright=~
     LocalAccelForward=~
     LocalAccelRight=~
     LocalVelocityRight=~
     SkydiveAimPitch=~
     SkydiveAimYaw=~
     DeployChuteAnimRate=~
     SkydiveDriftAnimRate=~
     SkydiveDriftAnimRateCurrent=~
     SkydiveFidgetAnimRate=~
     SkydiveFidgetAnimRateCurrent=~
     SkydiveAdditiveAlpha=~
     SkydiveDriftDelay=~
    ).bIsSkydiving)
  ]
 }
 Transition {
  path=AnimGraph.AnimGraphNode_StateMachine_0.Default.AnimStateNode_0.Locomotion.AnimGraphNode_StateMachine_1.Locomotion.AnimStateTransitionNode_7.Transition
  nodes [Result()]
 }
 Transition {
  path=AnimGraph.AnimGraphNode_StateMachine_0.Default.AnimStateNode_0.Locomotion.AnimGraphNode_StateMachine_1.Locomotion.AnimStateTransitionNode_9.Transition
  nodes [
   n1 = Get Skydiving(self=Get FortPlayerAnimBP())
   Result(bCanEnterTransition=NOT Boolean(A=Break Fort Anim Input Skydiving(FortAnimInput_Skydiving=@n1) -> (
      ParaGilderRootOffset=~
      ParaGliderRootRotationOverride=~
      bUseParaGlideRootModifier=~
      bIsSkydivingFromLaunchPad=~
      bIsSkydivingFromBus=~
      bIsInVortex=~
      bIsDiving=~
      bIsParachuteOpen=~
      bIsSkydiving=~
      bIsParachuteLeaning=~
      bIsSkydiveLeaning=~
      bIsLeaning=~
      bIsSkydiveDiveMode=~
      bPlaySkydiveDrift=~
      bSkydiveDriftDelayActive=~
      bSkydiveDriftAnimAllowed=~
      bIsGliderRight=~
      bIsGliderCenter=~
      bIsGliderLeft=~
      bIsGliderForward=~
      bIsGliderBack=~
      bHasDeployedAndIsStandingUpright=~
      LocalAccelForward=~
      LocalAccelRight=~
      LocalVelocityRight=~
      SkydiveAimPitch=~
      SkydiveAimYaw=~
      DeployChuteAnimRate=~
      SkydiveDriftAnimRate=~
      SkydiveDriftAnimRateCurrent=~
      SkydiveFidgetAnimRate=~
      SkydiveFidgetAnimRateCurrent=~
      SkydiveAdditiveAlpha=~
      SkydiveDriftDelay=~
     ).bIsSkydiving))
  ]
 }
 CowerEnd {
  path=AnimGraph.AnimGraphNode_StateMachine_0.Default.AnimStateNode_10.CowerEnd
  nodes [Output Animation Pose(Result=Sequence Player 'Cower_A_Stop'!(bLoopAnimation:bool=false) -> (Pose) {sequence=/BRCosmetics/Animation/Game/Pets/Cosmos/Backpack/Cower/Cower_A_Stop})]
 }
 UnderWaterStop {
  path=AnimGraph.AnimGraphNode_StateMachine_0.Default.AnimStateNode_11.UnderWaterStop
  nodes [
   n1 = Blend Poses by int!(
    BlendPose_0=Sequence Player 'UnderWaterStop'!(bLoopAnimation:bool=false) -> (Pose) {sequence=/BRCosmetics/Animation/Game/Pets/Cosmos/Backpack/UnderWater/UnderWaterStop}
    BlendPose_1=Sequence Player 'UnderWaterStop_B'!(bLoopAnimation:bool=false) -> (Pose) {sequence=/BRCosmetics/Animation/Game/Pets/Cosmos/Backpack/UnderWater/UnderWaterStop_B}
    BlendTime_0=0.1
    BlendTime_1=0.1
    ActiveChildIndex=Get UnderWaterStopSelection() -> (UnderWaterStopSelection:int=0)
   ) -> (Pose) {blendType=Linear, blendPose [{linkId=-1, sourceLinkId=-1}, {linkId=-1, sourceLinkId=-1}], blendTime [0.1,0.1]}
   Output Animation Pose(Result=@n1)
  ]
 }
 Hunker {
  path=AnimGraph.AnimGraphNode_StateMachine_0.Default.AnimStateNode_2.Hunker
  nodes [
   AnimGraphNode_StateMachine_0 = Hunker_SM() {is=title AnimGraphNode_StateMachine}
   Output Animation Pose(Result=@AnimGraphNode_StateMachine_0)
  ]
 }
 Hunker_SM {
  entry=HunkerStart
  path=AnimGraph.AnimGraphNode_StateMachine_0.Default.AnimStateNode_2.Hunker.AnimGraphNode_StateMachine_0.Hunker_SM
  ids=id
  nodes [
   AnimStateNode_0 = HunkerStart() {is=title AnimStateNode}
   AnimStateNode_1 = HunkerLoop() {is=title AnimStateNode}
   AnimStateTransitionNode_4 = AnimStateTransitionNode() {bAutomaticRuleBasedOnSequencePlayerInState=true}
  ]
  links [{
    dst=HunkerLoop
    rule=AnimGraph.AnimGraphNode_StateMachine_0.Default.AnimStateNode_2.Hunker.AnimGraphNode_StateMachine_0.Hunker_SM.AnimStateTransitionNode_4.Transition
    src=HunkerStart
    via=AnimStateTransitionNode_4
   }]
 }
 HunkerStart {
  path=AnimGraph.AnimGraphNode_StateMachine_0.Default.AnimStateNode_2.Hunker.AnimGraphNode_StateMachine_0.Hunker_SM.AnimStateNode_0.HunkerStart
  nodes [Output Animation Pose(Result=Sequence Player 'HunkerStart'() {sequence=/BRCosmetics/Animation/Game/Pets/Cosmos/Backpack/Idle/HunkerStart})]
 }
 HunkerLoop {
  path=AnimGraph.AnimGraphNode_StateMachine_0.Default.AnimStateNode_2.Hunker.AnimGraphNode_StateMachine_0.Hunker_SM.AnimStateNode_1.HunkerLoop
  nodes [Output Animation Pose(Result=Sequence Player 'HunkerLoop'() {sequence=/BRCosmetics/Animation/Game/Pets/Cosmos/Backpack/Idle/HunkerLoop})]
 }
 Transition {
  path=AnimGraph.AnimGraphNode_StateMachine_0.Default.AnimStateNode_2.Hunker.AnimGraphNode_StateMachine_0.Hunker_SM.AnimStateTransitionNode_4.Transition
  nodes [Result()]
 }
 HunkerStop {
  path=AnimGraph.AnimGraphNode_StateMachine_0.Default.AnimStateNode_4.HunkerStop
  nodes [Output Animation Pose(Result=Sequence Player 'HunkerStop'() {sequence=/BRCosmetics/Animation/Game/Pets/Cosmos/Backpack/Idle/HunkerStop})]
 }
 UnderWater {
  path=AnimGraph.AnimGraphNode_StateMachine_0.Default.AnimStateNode_5.UnderWater
  nodes [
   AnimGraphNode_StateMachine_0 = UnderWater() {is=title AnimGraphNode_StateMachine}
   Output Animation Pose(Result=@AnimGraphNode_StateMachine_0)
  ]
 }
 UnderWater {
  entry=UnderWaterStart
  path=AnimGraph.AnimGraphNode_StateMachine_0.Default.AnimStateNode_5.UnderWater.AnimGraphNode_StateMachine_0.UnderWater
  ids=id
  nodes [
   AnimStateNode_0 = UnderWaterStart() {is=title AnimStateNode}
   AnimStateNode_1 = UnderWaterLoop() {is=title AnimStateNode}
   AnimStateTransitionNode_3 = AnimStateTransitionNode() {bAutomaticRuleBasedOnSequencePlayerInState=true}
  ]
  links [{
    dst=UnderWaterLoop
    rule=AnimGraph.AnimGraphNode_StateMachine_0.Default.AnimStateNode_5.UnderWater.AnimGraphNode_StateMachine_0.UnderWater.AnimStateTransitionNode_3.Transition
    src=UnderWaterStart
    via=AnimStateTransitionNode_3
   }]
 }
 UnderWaterStart {
  path=AnimGraph.AnimGraphNode_StateMachine_0.Default.AnimStateNode_5.UnderWater.AnimGraphNode_StateMachine_0.UnderWater.AnimStateNode_0.UnderWaterStart
  nodes [Output Animation Pose(Result=Sequence Player 'UnderWaterStart'!(bLoopAnimation:bool=false) -> (Pose) {sequence=/BRCosmetics/Animation/Game/Pets/Cosmos/Backpack/UnderWater/UnderWaterStart})]
 }
 UnderWaterLoop {
  path=AnimGraph.AnimGraphNode_StateMachine_0.Default.AnimStateNode_5.UnderWater.AnimGraphNode_StateMachine_0.UnderWater.AnimStateNode_1.UnderWaterLoop
  nodes [
   AnimGraphNode_StateMachine_0 = UWStartADD() {is=title AnimGraphNode_StateMachine}
   Output Animation Pose(Result=Apply Additive(
     Base=Sequence Player 'UnderWaterLoop'() {sequence=/BRCosmetics/Animation/Game/Pets/Cosmos/Backpack/UnderWater/UnderWaterLoop}
     Additive=@AnimGraphNode_StateMachine_0
    ) {alphaBoolBlend {blendOption=Linear}})
  ]
 }
 UWStartADD {
  entry=Identity
  path=AnimGraph.AnimGraphNode_StateMachine_0.Default.AnimStateNode_5.UnderWater.AnimGraphNode_StateMachine_0.UnderWater.AnimStateNode_1.UnderWaterLoop.AnimGraphNode_StateMachine_0.UWStartADD
  ids=id
  nodes [
   AnimStateNode_0 = UWStartADD() {is=title AnimStateNode}
   AnimStateNode_1 = Identity() {is=title AnimStateNode}
   AnimStateTransitionNode_10 = AnimStateTransitionNode()
   AnimStateTransitionNode_11 = AnimStateTransitionNode()
  ]
  links [
   {
    dst=UWStartADD
    rule=AnimGraph.AnimGraphNode_StateMachine_0.Default.AnimStateNode_5.UnderWater.AnimGraphNode_StateMachine_0.UnderWater.AnimStateNode_1.UnderWaterLoop.AnimGraphNode_StateMachine_0.UWStartADD.AnimStateTransitionNode_10.Transition
    src=Identity
    via=AnimStateTransitionNode_10
   }
   {
    dst=Identity
    rule=AnimGraph.AnimGraphNode_StateMachine_0.Default.AnimStateNode_5.UnderWater.AnimGraphNode_StateMachine_0.UnderWater.AnimStateNode_1.UnderWaterLoop.AnimGraphNode_StateMachine_0.UWStartADD.AnimStateTransitionNode_11.Transition
    src=UWStartADD
    via=AnimStateTransitionNode_11
   }
  ]
 }
 UWStartADD {
  path=AnimGraph.AnimGraphNode_StateMachine_0.Default.AnimStateNode_5.UnderWater.AnimGraphNode_StateMachine_0.UnderWater.AnimStateNode_1.UnderWaterLoop.AnimGraphNode_StateMachine_0.UWStartADD.AnimStateNode_0.UWStartADD
  nodes [Output Animation Pose(Result="Sequence Player (additive) 'UnderWaterEnterAdditive'"!(bLoopAnimation:bool=false) -> (Pose) {sequence=/BRCosmetics/Animation/Game/Pets/Cosmos/Backpack/UnderWater/UnderWaterEnterAdditive})]
 }
 Identity {
  path=AnimGraph.AnimGraphNode_StateMachine_0.Default.AnimStateNode_5.UnderWater.AnimGraphNode_StateMachine_0.UnderWater.AnimStateNode_1.UnderWaterLoop.AnimGraphNode_StateMachine_0.UWStartADD.AnimStateNode_1.Identity
  nodes [Output Animation Pose(Result=Additive Identity Pose())]
 }
 Transition {
  path=AnimGraph.AnimGraphNode_StateMachine_0.Default.AnimStateNode_5.UnderWater.AnimGraphNode_StateMachine_0.UnderWater.AnimStateNode_1.UnderWaterLoop.AnimGraphNode_StateMachine_0.UWStartADD.AnimStateTransitionNode_10.Transition
  nodes [
   Result(bCanEnterTransition=Get EnteredWater() -> (EnteredWater:bool))
   AND Boolean(A=Get IsUnderWater() -> (IsUnderWater:bool), B=NOT Boolean(A=Get IsUnderWaterLastTick() -> (IsUnderWaterLastTick:bool)))
  ]
 }
 Transition {
  path=AnimGraph.AnimGraphNode_StateMachine_0.Default.AnimStateNode_5.UnderWater.AnimGraphNode_StateMachine_0.UnderWater.AnimStateNode_1.UnderWaterLoop.AnimGraphNode_StateMachine_0.UWStartADD.AnimStateTransitionNode_11.Transition
  nodes [Result(bCanEnterTransition=NOT Boolean(A=Get IsUnderWater()))]
 }
 Transition {
  path=AnimGraph.AnimGraphNode_StateMachine_0.Default.AnimStateNode_5.UnderWater.AnimGraphNode_StateMachine_0.UnderWater.AnimStateTransitionNode_3.Transition
  nodes [Result()]
 }
 Jump {
  path=AnimGraph.AnimGraphNode_StateMachine_0.Default.AnimStateNode_6.Jump
  nodes [
   AnimGraphNode_StateMachine_0 = Jump_SM() {is=title AnimGraphNode_StateMachine}
   Output Animation Pose(Result=@AnimGraphNode_StateMachine_0)
  ]
 }
 Jump_SM {
  entry=Jump
  path=AnimGraph.AnimGraphNode_StateMachine_0.Default.AnimStateNode_6.Jump.AnimGraphNode_StateMachine_0.Jump_SM
  ids=id
  nodes [
   AnimStateNode_0 = Jump()
   AnimStateNode_1 = Jump_Loop() {is=title AnimStateNode}
   AnimStateTransitionNode_0 = AnimStateTransitionNode() {bAutomaticRuleBasedOnSequencePlayerInState=true}
  ]
  links [{
    dst=Jump_Loop
    rule=AnimGraph.AnimGraphNode_StateMachine_0.Default.AnimStateNode_6.Jump.AnimGraphNode_StateMachine_0.Jump_SM.AnimStateTransitionNode_0.Transition
    src=Jump
    via=AnimStateTransitionNode_0
   }]
 }
 Jump {
  path=AnimGraph.AnimGraphNode_StateMachine_0.Default.AnimStateNode_6.Jump.AnimGraphNode_StateMachine_0.Jump_SM.AnimStateNode_0.Jump
  nodes [Output Animation Pose(Result=Sequence Player 'Jump'() {sequence=/BRCosmetics/Animation/Game/Pets/Cosmos/Backpack/Jump/Jump})]
 }
 Jump_Loop {
  path=AnimGraph.AnimGraphNode_StateMachine_0.Default.AnimStateNode_6.Jump.AnimGraphNode_StateMachine_0.Jump_SM.AnimStateNode_1.Jump_Loop
  nodes [Output Animation Pose(Result=Sequence Player 'Jump_Loop'() {sequence=/BRCosmetics/Animation/Game/Pets/Cosmos/Backpack/Jump/Jump_Loop})]
 }
 Transition {
  path=AnimGraph.AnimGraphNode_StateMachine_0.Default.AnimStateNode_6.Jump.AnimGraphNode_StateMachine_0.Jump_SM.AnimStateTransitionNode_0.Transition
  nodes [Result()]
 }
 JumpLand {
  path=AnimGraph.AnimGraphNode_StateMachine_0.Default.AnimStateNode_7.JumpLand
  nodes [Output Animation Pose(Result=Sequence Player 'Jump_Land'() {sequence=/BRCosmetics/Animation/Game/Pets/Cosmos/Backpack/Jump/Jump_Land})]
 }
 Cower {
  path=AnimGraph.AnimGraphNode_StateMachine_0.Default.AnimStateNode_8.Cower
  nodes [
   AnimGraphNode_StateMachine_0 = Cower() {is=title AnimGraphNode_StateMachine}
   Output Animation Pose(Result=@AnimGraphNode_StateMachine_0)
  ]
 }
 Cower {
  entry=CowerStart
  path=AnimGraph.AnimGraphNode_StateMachine_0.Default.AnimStateNode_8.Cower.AnimGraphNode_StateMachine_0.Cower
  ids=id
  nodes [
   AnimStateNode_0 = CowerStart() {is=title AnimStateNode}
   AnimStateNode_1 = CowerLoop() {is=title AnimStateNode, stateFullyBlended {guid=#7, notifyName=CowerLoop_FullyBlended}}
   AnimStateTransitionNode_7 = AnimStateTransitionNode() {bAutomaticRuleBasedOnSequencePlayerInState=true}
  ]
  links [{
    dst=CowerLoop
    rule=AnimGraph.AnimGraphNode_StateMachine_0.Default.AnimStateNode_8.Cower.AnimGraphNode_StateMachine_0.Cower.AnimStateTransitionNode_7.Transition
    src=CowerStart
    via=AnimStateTransitionNode_7
   }]
 }
 CowerStart {
  path=AnimGraph.AnimGraphNode_StateMachine_0.Default.AnimStateNode_8.Cower.AnimGraphNode_StateMachine_0.Cower.AnimStateNode_0.CowerStart
  nodes [
   n1 = Blend Poses by int(
    BlendPose_0=Sequence Player 'Cower_B_Start'!(bLoopAnimation:bool=false) -> (Pose) {sequence=/BRCosmetics/Animation/Game/Pets/Cosmos/Backpack/Cower/Cower_B_Start}
    BlendPose_1=Sequence Player 'Cower_A_Start'!(bLoopAnimation:bool=false) -> (Pose) {sequence=/BRCosmetics/Animation/Game/Pets/Cosmos/Backpack/Cower/Cower_A_Start}
    BlendTime_0=0
    BlendTime_1=0
    ActiveChildIndex=Get CowerSelection() -> (CowerSelection:int=0)
   ) {blendType=Linear, blendPose [{linkId=-1, sourceLinkId=-1}, {linkId=-1, sourceLinkId=-1}], blendTime [0.1,0.1]}
   Output Animation Pose(Result=@n1)
  ]
 }
 CowerLoop {
  path=AnimGraph.AnimGraphNode_StateMachine_0.Default.AnimStateNode_8.Cower.AnimGraphNode_StateMachine_0.Cower.AnimStateNode_1.CowerLoop
  nodes [Output Animation Pose(Result=Sequence Player 'Cower_A_Loop'() {sequence=/BRCosmetics/Animation/Game/Pets/Cosmos/Backpack/Cower/Cower_A_Loop})]
 }
 Transition {
  path=AnimGraph.AnimGraphNode_StateMachine_0.Default.AnimStateNode_8.Cower.AnimGraphNode_StateMachine_0.Cower.AnimStateTransitionNode_7.Transition
  nodes [Result()]
 }
 Transition {
  path=AnimGraph.AnimGraphNode_StateMachine_0.Default.AnimStateTransitionNode_1.Transition
  nodes [Result(bCanEnterTransition=Get bIsPlayingMeleeAnim!(self:Fort Player Anim Instance Object Reference=Get FortPlayerAnimBP()) -> (bIsPlayingMeleeAnim))]
 }
 Transition {
  path=AnimGraph.AnimGraphNode_StateMachine_0.Default.AnimStateTransitionNode_10.Transition
  nodes [Result(bCanEnterTransition=Get bIsSurfaceSwimming!(self:Fort Player Anim Instance Object Reference=Get FortPlayerAnimBP()) -> (bIsSurfaceSwimming))]
 }
 Transition {
  path=AnimGraph.AnimGraphNode_StateMachine_0.Default.AnimStateTransitionNode_11.Transition
  nodes [Result(bCanEnterTransition=Get IsUnderWaterHunker() -> (IsUnderWaterHunker:bool))]
 }
 Transition {
  path=AnimGraph.AnimGraphNode_StateMachine_0.Default.AnimStateTransitionNode_12.Transition
  nodes [Result(bCanEnterTransition=Get IsUnderWaterHunker())]
 }
 Transition {
  path=AnimGraph.AnimGraphNode_StateMachine_0.Default.AnimStateTransitionNode_13.Transition
  nodes [Result(bCanEnterTransition=Get IsCower() -> (IsCower:bool))]
 }
 Transition {path=AnimGraph.AnimGraphNode_StateMachine_0.Default.AnimStateTransitionNode_14.Transition, nodes [Result(bCanEnterTransition=Get IsCower())]}
 Transition {
  path=AnimGraph.AnimGraphNode_StateMachine_0.Default.AnimStateTransitionNode_15.Transition
  nodes [
   n1 = Get IsCower()
   n2 = Get FortPlayerAnimBP()
   Result(bCanEnterTransition=AND Boolean(A=Get bIsJumping!(self:Fort Anim Instance Object Reference=@n2) -> (bIsJumping), B=NOT Boolean(A=@n1)))
  ]
 }
 Transition {
  path=AnimGraph.AnimGraphNode_StateMachine_0.Default.AnimStateTransitionNode_16.Transition
  nodes [Result(bCanEnterTransition=Get IsMovingOnGround() -> (IsMovingOnGround:bool))]
 }
 Transition {path=AnimGraph.AnimGraphNode_StateMachine_0.Default.AnimStateTransitionNode_17.Transition, nodes [Result()]}
 Transition {
  path=AnimGraph.AnimGraphNode_StateMachine_0.Default.AnimStateTransitionNode_18.Transition
  nodes [Result(bCanEnterTransition=NOT Boolean(A=Get IsUnderWaterHunker()))]
 }
 Transition {path=AnimGraph.AnimGraphNode_StateMachine_0.Default.AnimStateTransitionNode_19.Transition, nodes [Result()]}
 Transition {
  path=AnimGraph.AnimGraphNode_StateMachine_0.Default.AnimStateTransitionNode_20.Transition
  nodes [Result(bCanEnterTransition=NOT Boolean(A=Get IsCower()))]
 }
 Transition {
  path=AnimGraph.AnimGraphNode_StateMachine_0.Default.AnimStateTransitionNode_21.Transition
  nodes [Result(bCanEnterTransition=Get IsUnderWaterHunker())]
 }
 Transition {
  path=AnimGraph.AnimGraphNode_StateMachine_0.Default.AnimStateTransitionNode_4.Transition
  nodes [
   n1 = Get IsCower()
   Result(bCanEnterTransition=@n1)
   n2 = Get FortPlayerAnimBP()
   AND Boolean(A=@n1, B=NOT Boolean(A=Get bIsDBNO!(self:Fort Player Anim Instance Object Reference=@n2) -> (bIsDBNO)))
  ]
 }
 Transition {path=AnimGraph.AnimGraphNode_StateMachine_0.Default.AnimStateTransitionNode_7.Transition, nodes [Result()]}
 Transition {
  path=AnimGraph.AnimGraphNode_StateMachine_0.Default.AnimStateTransitionNode_8.Transition
  nodes [
   n1 = Get FortPlayerAnimBP()
   Result(bCanEnterTransition=NOT Boolean(A=Get bIsPlayingMeleeAnim(self=@n1)))
  ]
 }
 Transition {path=AnimGraph.AnimGraphNode_StateMachine_0.Default.AnimStateTransitionNode_9.Transition, nodes [Result()]}
 FrontEnd {
  entry=Idle
  path=AnimGraph.AnimGraphNode_StateMachine_1.FrontEnd
  ids=id
  nodes [
   AnimStateNode_0 = Idle()
   AnimStateNode_1 = Skydive() {is=title AnimStateNode}
   AnimStateTransitionNode_7 = AnimStateTransitionNode() {bAutomaticRuleBasedOnSequencePlayerInState=true}
  ]
  links [{
    dst=Skydive
    rule=AnimGraph.AnimGraphNode_StateMachine_1.FrontEnd.AnimStateTransitionNode_7.Transition
    src=Idle
    via=AnimStateTransitionNode_7
   }]
 }
 Idle {
  path=AnimGraph.AnimGraphNode_StateMachine_1.FrontEnd.AnimStateNode_0.Idle
  nodes [Output Animation Pose(Result=Sequence Player 'Idle'() {sequence=&Idle})]
 }
 Skydive {path=AnimGraph.AnimGraphNode_StateMachine_1.FrontEnd.AnimStateNode_1.Skydive, nodes [Output Animation Pose()]}
 Transition {path=AnimGraph.AnimGraphNode_StateMachine_1.FrontEnd.AnimStateTransitionNode_7.Transition, nodes [Result()]}
 FrontEndSM {
  entry=FE_Idle
  path=AnimGraph.AnimGraphNode_StateMachine_2.FrontEndSM
  ids=id
  nodes [
   AnimStateNode_0 = FE_Idle() {is=title AnimStateNode}
   AnimStateNode_10 = CowerEnd()
   AnimStateNode_8 = FE_Cower() {is=title AnimStateNode}
   AnimStateTransitionNode_1 = AnimStateTransitionNode()
   AnimStateTransitionNode_7 = AnimStateTransitionNode()
   AnimStateTransitionNode_8 = AnimStateTransitionNode() {bAutomaticRuleBasedOnSequencePlayerInState=true}
  ]
  links [
   {
    dst=FE_Idle
    rule=AnimGraph.AnimGraphNode_StateMachine_2.FrontEndSM.AnimStateTransitionNode_8.Transition
    src=CowerEnd
    via=AnimStateTransitionNode_8
   }
   {
    dst=CowerEnd
    rule=AnimGraph.AnimGraphNode_StateMachine_2.FrontEndSM.AnimStateTransitionNode_7.Transition
    src=FE_Cower
    via=AnimStateTransitionNode_7
   }
   {
    dst=FE_Cower
    rule=AnimGraph.AnimGraphNode_StateMachine_2.FrontEndSM.AnimStateTransitionNode_1.Transition
    src=FE_Idle
    via=AnimStateTransitionNode_1
   }
  ]
 }
 FE_Idle {
  path=AnimGraph.AnimGraphNode_StateMachine_2.FrontEndSM.AnimStateNode_0.FE_Idle
  nodes [
   AnimGraphNode_StateMachine_0 = FE_NonSkyDive() {is=title AnimGraphNode_StateMachine}
   Output Animation Pose(Result=@AnimGraphNode_StateMachine_0)
  ]
 }
 FE_NonSkyDive {
  entry=FE_Idle
  path=AnimGraph.AnimGraphNode_StateMachine_2.FrontEndSM.AnimStateNode_0.FE_Idle.AnimGraphNode_StateMachine_0.FE_NonSkyDive
  ids=id
  nodes [
   AnimStateNode_0 = FE_Idle()
   AnimStateNode_1 = FE_Break() {is=title AnimStateNode, stateEntered {guid=#8, notifyName=Entered FE_Break}}
   AnimStateNode_2 = FE_MoveStart() {is=title AnimStateNode}
   AnimStateNode_3 = FE_MoveEnd() {is=title AnimStateNode}
   AnimStateTransitionNode_2 = AnimStateTransitionNode()
   AnimStateTransitionNode_3 = AnimStateTransitionNode() {bAutomaticRuleBasedOnSequencePlayerInState=true}
   AnimStateTransitionNode_4 = AnimStateTransitionNode() {bAutomaticRuleBasedOnSequencePlayerInState=true}
   AnimStateTransitionNode_5 = AnimStateTransitionNode() {bAutomaticRuleBasedOnSequencePlayerInState=true}
  ]
  links [
   {
    dst=FE_MoveEnd
    rule=AnimGraph.AnimGraphNode_StateMachine_2.FrontEndSM.AnimStateNode_0.FE_Idle.AnimGraphNode_StateMachine_0.FE_NonSkyDive.AnimStateTransitionNode_3.AnimationTransitionGraph_0
    src=FE_Break
    via=AnimStateTransitionNode_3
   }
   {
    dst=FE_MoveStart
    rule=AnimGraph.AnimGraphNode_StateMachine_2.FrontEndSM.AnimStateNode_0.FE_Idle.AnimGraphNode_StateMachine_0.FE_NonSkyDive.AnimStateTransitionNode_2.AnimationTransitionGraph_0
    src=FE_Idle
    via=AnimStateTransitionNode_2
   }
   {
    dst=FE_Idle
    rule=AnimGraph.AnimGraphNode_StateMachine_2.FrontEndSM.AnimStateNode_0.FE_Idle.AnimGraphNode_StateMachine_0.FE_NonSkyDive.AnimStateTransitionNode_5.AnimationTransitionGraph_0
    src=FE_MoveEnd
    via=AnimStateTransitionNode_5
   }
   {
    dst=FE_Break
    rule=AnimGraph.AnimGraphNode_StateMachine_2.FrontEndSM.AnimStateNode_0.FE_Idle.AnimGraphNode_StateMachine_0.FE_NonSkyDive.AnimStateTransitionNode_4.AnimationTransitionGraph_0
    src=FE_MoveStart
    via=AnimStateTransitionNode_4
   }
  ]
 }
 FE_Idle {
  path=AnimGraph.AnimGraphNode_StateMachine_2.FrontEndSM.AnimStateNode_0.FE_Idle.AnimGraphNode_StateMachine_0.FE_NonSkyDive.AnimStateNode_0.FE_Idle
  nodes [Output Animation Pose(Result=Sequence Player 'Idle'() {sequence=&Idle})]
 }
 FE_Break {
  path=AnimGraph.AnimGraphNode_StateMachine_2.FrontEndSM.AnimStateNode_0.FE_Idle.AnimGraphNode_StateMachine_0.FE_NonSkyDive.AnimStateNode_1.FE_Break
  nodes [
   n1 = Blend Poses by int!(
    BlendPose_0=Sequence Player 'Locomotion_Break1'(bLoopAnimation=false) {sequence=&Locomotion_Break1}
    BlendPose_1=Sequence Player 'Locomotion_Break2'(bLoopAnimation=false) {sequence=&Locomotion_Break2}
    BlendPose_2:Pose Link Structure=Sequence Player 'Locomotion_Break3'(bLoopAnimation=false) {sequence=&Locomotion_Break3}
    BlendTime_0=0.1
    BlendTime_1=0.1
    BlendTime_2:float=0.1
    ActiveChildIndex=Get Choose_LocoBreak_Index() -> (Choose_LocoBreak_Index=0)
   ) -> (Pose) {
    blendType=Linear
    blendPose [{linkId=-1, sourceLinkId=-1}, {linkId=-1, sourceLinkId=-1}, {linkId=-1, sourceLinkId=-1}]
    blendTime [0.1,0.1,0.1]
   }
   Output Animation Pose(Result=@n1)
  ]
 }
 FE_MoveStart {
  path=AnimGraph.AnimGraphNode_StateMachine_2.FrontEndSM.AnimStateNode_0.FE_Idle.AnimGraphNode_StateMachine_0.FE_NonSkyDive.AnimStateNode_2.FE_MoveStart
  nodes [Output Animation Pose(Result=Sequence Player 'LocomotionStart'(bLoopAnimation=false) {sequence=&LocomotionStart})]
 }
 FE_MoveEnd {
  path=AnimGraph.AnimGraphNode_StateMachine_2.FrontEndSM.AnimStateNode_0.FE_Idle.AnimGraphNode_StateMachine_0.FE_NonSkyDive.AnimStateNode_3.FE_MoveEnd
  nodes [Output Animation Pose(Result=Sequence Player 'LocomotionStop'(bLoopAnimation=false) {sequence=&LocomotionStop})]
 }
 AnimationTransitionGraph_0 {
  path=AnimGraph.AnimGraphNode_StateMachine_2.FrontEndSM.AnimStateNode_0.FE_Idle.AnimGraphNode_StateMachine_0.FE_NonSkyDive.AnimStateTransitionNode_2.AnimationTransitionGraph_0
  nodes [Result(bCanEnterTransition=Get AllowLocomotionBreak())]
 }
 AnimationTransitionGraph_0 {
  path=AnimGraph.AnimGraphNode_StateMachine_2.FrontEndSM.AnimStateNode_0.FE_Idle.AnimGraphNode_StateMachine_0.FE_NonSkyDive.AnimStateTransitionNode_3.AnimationTransitionGraph_0
  nodes [Result()]
 }
 AnimationTransitionGraph_0 {
  path=AnimGraph.AnimGraphNode_StateMachine_2.FrontEndSM.AnimStateNode_0.FE_Idle.AnimGraphNode_StateMachine_0.FE_NonSkyDive.AnimStateTransitionNode_4.AnimationTransitionGraph_0
  nodes [Result()]
 }
 AnimationTransitionGraph_0 {
  path=AnimGraph.AnimGraphNode_StateMachine_2.FrontEndSM.AnimStateNode_0.FE_Idle.AnimGraphNode_StateMachine_0.FE_NonSkyDive.AnimStateTransitionNode_5.AnimationTransitionGraph_0
  nodes [Result()]
 }
 CowerEnd {
  path=AnimGraph.AnimGraphNode_StateMachine_2.FrontEndSM.AnimStateNode_10.CowerEnd
  nodes [Output Animation Pose(Result=Sequence Player 'Cower_A_Stop'(bLoopAnimation=false) {sequence=&Cower_A_Stop})]
 }
 FE_Cower {
  path=AnimGraph.AnimGraphNode_StateMachine_2.FrontEndSM.AnimStateNode_8.FE_Cower
  nodes [
   AnimGraphNode_StateMachine_0 = Cower()
   Output Animation Pose(Result=@AnimGraphNode_StateMachine_0)
  ]
 }
 Cower {
  entry=FE_CowerStart
  path=AnimGraph.AnimGraphNode_StateMachine_2.FrontEndSM.AnimStateNode_8.FE_Cower.AnimGraphNode_StateMachine_0.Cower
  ids=id
  nodes [
   AnimStateNode_0 = FE_CowerStart() {is=title AnimStateNode}
   AnimStateNode_1 = FE_CowerLoop() {is=title AnimStateNode, stateFullyBlended {guid=#7, notifyName=FE_CowerLoop_FullyBlended}}
   AnimStateTransitionNode_7 = AnimStateTransitionNode() {bAutomaticRuleBasedOnSequencePlayerInState=true}
  ]
  links [{
    dst=FE_CowerLoop
    rule=AnimGraph.AnimGraphNode_StateMachine_2.FrontEndSM.AnimStateNode_8.FE_Cower.AnimGraphNode_StateMachine_0.Cower.AnimStateTransitionNode_7.Transition
    src=FE_CowerStart
    via=AnimStateTransitionNode_7
   }]
 }
 FE_CowerStart {
  path=AnimGraph.AnimGraphNode_StateMachine_2.FrontEndSM.AnimStateNode_8.FE_Cower.AnimGraphNode_StateMachine_0.Cower.AnimStateNode_0.FE_CowerStart
  nodes [
   n1 = Blend Poses by int!(
    BlendPose_0=Sequence Player 'Cower_B_Start'(bLoopAnimation=false) {sequence=&Cower_B_Start}
    BlendPose_1=Sequence Player 'Cower_A_Start'(bLoopAnimation=false) {sequence=&Cower_A_Start}
    BlendTime_0=0
    BlendTime_1=0
    ActiveChildIndex=Get CowerSelection() -> (CowerSelection=0)
   ) -> (Pose) {blendType=Linear, blendPose [{linkId=-1, sourceLinkId=-1}, {linkId=-1, sourceLinkId=-1}], blendTime [0.1,0.1]}
   Output Animation Pose(Result=@n1)
  ]
 }
 FE_CowerLoop {
  path=AnimGraph.AnimGraphNode_StateMachine_2.FrontEndSM.AnimStateNode_8.FE_Cower.AnimGraphNode_StateMachine_0.Cower.AnimStateNode_1.FE_CowerLoop
  nodes [Output Animation Pose(Result=Sequence Player 'Cower_A_Loop'() {sequence=&Cower_A_Loop})]
 }
 Transition {
  path=AnimGraph.AnimGraphNode_StateMachine_2.FrontEndSM.AnimStateNode_8.FE_Cower.AnimGraphNode_StateMachine_0.Cower.AnimStateTransitionNode_7.Transition
  nodes [Result()]
 }
 Transition {
  path=AnimGraph.AnimGraphNode_StateMachine_2.FrontEndSM.AnimStateTransitionNode_1.Transition
  nodes [Result(bCanEnterTransition=Get IsCower())]
 }
 Transition {
  path=AnimGraph.AnimGraphNode_StateMachine_2.FrontEndSM.AnimStateTransitionNode_7.Transition
  nodes [Result(bCanEnterTransition=NOT Boolean(A=Get IsCower()))]
 }
 Transition {path=AnimGraph.AnimGraphNode_StateMachine_2.FrontEndSM.AnimStateTransitionNode_8.Transition, nodes [Result()]}
 EventGraph {nodes [
   Delay(Duration=0.5)
   > Set DelayMoveStart(DelayMoveStart:bool=Get bIsMoving2D(self=Get FortPlayerAnimBP())) -> (Output_Get:bool)
   Event BlueprintUpdateAnimation()
   > n1 = Branch(Condition=Get bIsInFrontEnd())
   > Set IsCower(IsCower:bool=Break Fort Anim Input Skydiving(FortAnimInput_Skydiving=Get SkydivingState() -> (SkydivingState:Fort Anim Input Skydiving Structure)) -> (
     ParaGilderRootOffset=~
     ParaGliderRootRotationOverride=~
     bUseParaGlideRootModifier=~
     bIsSkydivingFromLaunchPad=~
     bIsSkydivingFromBus=~
     bIsInVortex=~
     bIsDiving=~
     bIsParachuteOpen=~
     bIsSkydiving=~
     bIsParachuteLeaning=~
     bIsSkydiveLeaning=~
     bIsLeaning=~
     bIsSkydiveDiveMode=~
     bPlaySkydiveDrift=~
     bSkydiveDriftDelayActive=~
     bSkydiveDriftAnimAllowed=~
     bIsGliderRight=~
     bIsGliderCenter=~
     bIsGliderLeft=~
     bIsGliderForward=~
     bIsGliderBack=~
     bHasDeployedAndIsStandingUpright=~
     LocalAccelForward=~
     LocalAccelRight=~
     LocalVelocityRight=~
     SkydiveAimPitch=~
     SkydiveAimYaw=~
     DeployChuteAnimRate=~
     SkydiveDriftAnimRate=~
     SkydiveDriftAnimRateCurrent=~
     SkydiveFidgetAnimRate=~
     SkydiveFidgetAnimRateCurrent=~
     SkydiveAdditiveAlpha=~
     SkydiveDriftDelay=~
    ).bIsSkydiving) -> (Output_Get:bool)
   Is Valid(exec=@n1.else, InputObject=Get FortPlayerPawn() -> (FortPlayerPawn:Fort Player Pawn Object Reference))
   > Is Valid(InputObject=Get FortPlayerAnimBP())
   > n2 = Branch(Condition=Get bIsInFrontEnd())
   n3 = Sequence!(execute=@n2.else) -> (then_0, then_1, then_2:Exec, then_3:Exec, then_4:Exec)
   K2Node_Composite_0 = Cower!(execute=@n3.then_1) -> (then_2:Exec) {is=title K2Node_Composite}
   > K2Node_Composite_1 = UnderWater() {is=title K2Node_Composite}
   n4 = Sequence!(execute=@n3.then_2) -> (then_0, then_1, then_2, then_3)
   Branch(execute=@n4.then_1, Condition=Get bIsTargeting())
   > Set HasTargettedInLastNSeconds_b(HasTargettedInLastNSeconds_b:bool=true) -> (Output_Get:bool)
   > RetriggerableDelay(Duration=2.5)
   > Set HasTargettedInLastNSeconds_b(HasTargettedInLastNSeconds_b=false)
   K2Node_Composite_2 = ValuesLastTick(execute=@n3.then_3) {is=title K2Node_Composite}
   Set TookDamage(execute=@n3.then_4, TookDamage:bool=false) -> (Output_Get:bool)
   n5 = Get FortPlayerAnimBP()
   Set OwnerLocomotionDirection_b(
    execute=@n3.then_0
    OwnerLocomotionDirection_b:double=Get LocalVelocityYawAngle!(self:Fort Player Anim Instance Object Reference=@n5) -> (LocalVelocityYawAngle:float)
   ) -> (Output_Get:double)
   > Set OwnerLocomotionSpeed_b(OwnerLocomotionSpeed_b:double=Get Speed2D!(self:Fort Player Anim Instance Object Reference=@n5) -> (Speed2D:float)) -> (Output_Get:double)
   n6 = Get FortPlayerPawn()
   > Set IsMovingOnGround(IsMovingOnGround:bool=IsMovingOnGround!(self:Nav Movement Component Object Reference=Get CharacterMovement!(self:Character Object Reference=@n6) -> (CharacterMovement:Character Movement Component Object Reference)) -> (ReturnValue:bool)) -> (Output_Get:bool)
   Event BlueprintInitializeAnimation()
   > Sequence!(execute) -> (then_0, then_1)
   > n7 = Cast To B_PetInst_Cosmos!(execute, Object=GetOwningActor()) -> (then, CastFailed, AsB Pet Inst Cosmos:B Pet Inst Cosmos Object Reference) {is=title K2Node_DynamicCast}
   > n8 = Set PetInstCosmos(PetInstCosmos:B Pet Inst Cosmos Object Reference=@n7) -> (Output_Get:B Pet Inst Cosmos Object Reference)
   > n9 = Cast To FortPlayerPawn(Object=GetOwner(self:Actor Object Reference=@n8))
   > Set FortPlayerPawn(FortPlayerPawn:Fort Player Pawn Object Reference=@n9) -> (Output_Get:Fort Player Pawn Object Reference)
   n10 = Get FortPlayerPawn()
   > n11 = Cast To M_MED_Carrier_Cosmos_Pack_AnimBP!(execute, Object=GetAnimInstance(self=GetSkeletalMeshForPartType(self=@n10, PartType=Backpack))) -> (then, CastFailed, AsM MED Carrier Cosmos Pack Anim BP:M MED Carrier Cosmos Pack Anim BP Object Reference) {is=title K2Node_DynamicCast}
   > Set CarrierAnimBP(CarrierAnimBP:M MED Carrier Cosmos Pack Anim BP Object Reference=@n11) -> (Output_Get:M MED Carrier Cosmos Pack Anim BP Object Reference)
   n12 = Get FortPlayerPawn()
   > n13 = Cast To Fortnite_M_Avg_Player_AnimBlueprint!(execute, Object=GetAnimInstance(self=Get Mesh(self=@n12))) -> (then, CastFailed, AsFortnite M Avg Player Anim Blueprint:Fortnite M Avg Player Anim Blueprint Object Reference) {is=title K2Node_DynamicCast}
   > Set FortPlayerAnimBP(FortPlayerAnimBP:Fortnite M Avg Player Anim Blueprint Object Reference=@n13) -> (Output_Get:Fortnite M Avg Player Anim Blueprint Object Reference)
   AnimNotify_TriggerLocomotionRandomizer() {is=title K2Node_Event}
   > n14 = Set AllowLocomotionBreak(AllowLocomotionBreak:bool=RandomBoolWithWeight!(Weight:float=1) -> (ReturnValue:bool)) -> (Output_Get:bool)
   > Branch(Condition=@n14)
   > Set Choose_LocoBreak_Index(Choose_LocoBreak_Index:int=RandomInteger!(Max:int=3) -> (ReturnValue:int=0)) -> (Output_Get:int=0)
   n15 = AnimNotify_Entered_SkyDiveBreak() {is=title K2Node_Event}
   Set AllowLocomotionBreak(execute=[@n15.then, @n17.then, @n20.then], AllowLocomotionBreak=false)
   AnimNotify_CarrierDoorsClose() {is=title K2Node_Event}
   n16 = Get CarrierAnimBP() -> (CarrierAnimBP:M MED Carrier Cosmos Pack Anim BP Object Reference)
   > CarrierDoorsClose(self:M MED Carrier Cosmos Pack Anim BP Object Reference=@n16)
   AnimNotify_CarrierDoorsOpen() {is=title K2Node_Event}
   > CarrierDoorsOpen(self:M MED Carrier Cosmos Pack Anim BP Object Reference=@n16)
   n17 = AnimNotify_Entered_RandomLocoBreak() {is=title K2Node_Event}
   n18 = Event BlueprintProcessAnimInputEvent()
   n19 = OR Boolean(
    A=IsNamed(self=@n18.AnimInputEvent, InEventName=aie.PlayerVitality.HealthDecreased)
    B=IsNamed(self=@n18.AnimInputEvent, InEventName=aie.PlayerVitality.ShieldsDecreased)
   )
   > Branch(Condition=@n19)
   > Set TookDamage(TookDamage=@n19)
   n20 = AnimNotify_Entered FE_Break() {is=title K2Node_Event}
   Branch()
   > Set StoppedMovingOrMeleeForLongEnoughTime_b(StoppedMovingOrMeleeForLongEnoughTime_b:bool=false) -> (Output_Get:bool)
   > RetriggerableDelay(Duration=0.2)
   > Set StoppedMovingOrMeleeForLongEnoughTime_b(StoppedMovingOrMeleeForLongEnoughTime_b=true)
   n21 = Get CarrierAnimBP()
   PrintString!(
    execute
    InString="To String (Boolean)"(InBool=IsNullItem!(ItemObject:Object Reference=@n21) -> (Is Null Item:bool))
    bPrintToScreen
    bPrintToLog
    TextColor
    Duration
    Key
   ) -> (then)
   AnimNotify_JumpEnded() {is=title K2Node_Event}
  ]}
 Cower {
  path=EventGraph.K2Node_Composite_0.Cower
  nodes [
   n1 = AnimNotify_FE_CowerLoop_FullyBlended() {is=title K2Node_Event}
   n2 = Branch(execute=[@n1.then, @n3.then], Condition="Equal (Integer)"(A=Get CowerSelection() -> (CowerSelection=0), B=0) {is=title K2Node_CallFunction})
   > Set CowerSelection(CowerSelection:int=1) -> (Output_Get:int=0)
   Set CowerSelection(execute=@n2.else, CowerSelection=RandomIntegerInRange!(Min:int=0, Max:int=1) -> (ReturnValue:int=0)) -> (Output_Get=0)
   n3 = AnimNotify_CowerLoop_FullyBlended() {is=title K2Node_Event}
   n4 = Get Skydiving(self=Get FortPlayerAnimBP())
   Branch(Condition=AND Boolean(
     A=Get IsCower()
     B=Break Fort Anim Input Skydiving(FortAnimInput_Skydiving=@n4) -> (
      ParaGilderRootOffset=~
      ParaGliderRootRotationOverride=~
      bUseParaGlideRootModifier=~
      bIsSkydivingFromLaunchPad=~
      bIsSkydivingFromBus=~
      bIsInVortex=~
      bIsDiving=~
      bIsParachuteOpen=~
      bIsSkydiving=~
      bIsParachuteLeaning=~
      bIsSkydiveLeaning=~
      bIsLeaning=~
      bIsSkydiveDiveMode=~
      bPlaySkydiveDrift=~
      bSkydiveDriftDelayActive=~
      bSkydiveDriftAnimAllowed=~
      bIsGliderRight=~
      bIsGliderCenter=~
      bIsGliderLeft=~
      bIsGliderForward=~
      bIsGliderBack=~
      bHasDeployedAndIsStandingUpright=~
      LocalAccelForward=~
      LocalAccelRight=~
      LocalVelocityRight=~
      SkydiveAimPitch=~
      SkydiveAimYaw=~
      DeployChuteAnimRate=~
      SkydiveDriftAnimRate=~
      SkydiveDriftAnimRateCurrent=~
      SkydiveFidgetAnimRate=~
      SkydiveFidgetAnimRateCurrent=~
      SkydiveAdditiveAlpha=~
      SkydiveDriftDelay=~
     ).bIsSkydiving
    ))
   > n5 = Sequence()
   > Do Once()
   > Set BeginSkyDiveTime(BeginSkyDiveTime:double=GetGameTimeInSeconds()) -> (Output_Get:double)
   n6 = GetGameTimeInSeconds()
   n7 = Get BeginSkyDiveTime() -> (BeginSkyDiveTime:double)
   Branch(
    execute=@n5.then_1
    Condition="float > float"(A=float - float(A=@n6, B=@n7) {is=title K2Node_CallFunction}, B=RandomFloatInRange(Min=8, Max=15)) {is=title K2Node_CallFunction}
   )
   > CarrierDoorsOpen(self=Get CarrierAnimBP())
   > Set BeginSkyDiveTime(BeginSkyDiveTime=GetGameTimeInSeconds())
   > Delay(Duration=4)
   > Branch(Condition=Get IsCower())
   > CarrierDoorsClose(self=Get CarrierAnimBP())
   Inputs()
   > n8 = Sequence!(execute) -> (then_0, then_1, then_2:Exec)
   n9 = Break Fort Anim Input Skydiving(FortAnimInput_Skydiving=Get Skydiving(self=Get FortPlayerAnimBP())) -> (
    ParaGilderRootOffset=~
    ParaGliderRootRotationOverride=~
    bUseParaGlideRootModifier=~
    bIsSkydivingFromLaunchPad=~
    bIsSkydivingFromBus=~
    bIsInVortex=~
    bIsDiving=~
    bIsParachuteOpen=~
    bIsSkydiving=~
    bIsParachuteLeaning=~
    bIsSkydiveLeaning=~
    bIsLeaning=~
    bIsSkydiveDiveMode=~
    bPlaySkydiveDrift=~
    bSkydiveDriftDelayActive=~
    bSkydiveDriftAnimAllowed=~
    bIsGliderRight=~
    bIsGliderCenter=~
    bIsGliderLeft=~
    bIsGliderForward=~
    bIsGliderBack=~
    bHasDeployedAndIsStandingUpright=~
    LocalAccelForward=~
    LocalAccelRight=~
    LocalVelocityRight=~
    SkydiveAimPitch=~
    SkydiveAimYaw=~
    DeployChuteAnimRate=~
    SkydiveDriftAnimRate=~
    SkydiveDriftAnimRateCurrent=~
    SkydiveFidgetAnimRate=~
    SkydiveFidgetAnimRateCurrent=~
    SkydiveAdditiveAlpha=~
    SkydiveDriftDelay=~
   )
   n10 = NOT Boolean(A=@n9.bIsParachuteOpen)
   n11 = Get FortPlayerAnimBP()
   > Branch(Condition=OR Boolean!(A=AND Boolean(A=@n10, B=@n9.bIsSkydiving), B=Get TookDamage() -> (TookDamage:bool), C:bool=Get bIsDBNO(self=@n11)) -> (ReturnValue))
   > Set IsCower(IsCower=true)
   > RetriggerableDelay(Duration=5)
   > Set IsCower(IsCower=false)
   Reroute Node(InputPin=@n8.then_1)
   Outputs!(then_2:Exec=@n8.then_2)
  ]
 }
 UnderWater {
  path=EventGraph.K2Node_Composite_1.UnderWater
  nodes [
   AnimNotify_UnderWaterEntered() {is=title K2Node_Event}
   > n1 = Flip Flop!(None:Exec) -> (A:Exec, B:Exec, IsA:bool=~) {is=title K2Node_MacroInstance}
   > Set UnderWaterStopSelection(UnderWaterStopSelection:int=0) -> (Output_Get:int=0)
   Set UnderWaterStopSelection(execute=@n1.B, UnderWaterStopSelection=1) -> (Output_Get=0)
   Inputs()
   > n2 = Sequence!(execute) -> (then_0, then_1)
   n3 = Get FortPlayerPawn()
   > n4 = Branch(Condition="float > float"(A=GetWaterImmersionDepth!(self:Fort Pawn Object Reference=@n3) -> (ReturnValue:float), B=140))
   > Sequence()
   > Set IsUnderWater(IsUnderWater:bool=true) -> (Output_Get:bool)
   > Set IsUnderWaterHunker(IsUnderWaterHunker:bool=true) -> (Output_Get:bool)
   > RetriggerableDelay(Duration=1)
   > Set IsUnderWaterHunker(IsUnderWaterHunker=false)
   Set IsUnderWater(execute=@n4.else, IsUnderWater=false)
   n5 = Get IsUnderWaterLastTick()
   Set EnteredWater(execute=@n2.then_1, EnteredWater:bool=AND Boolean(A=Get IsUnderWater(), B=NOT Boolean(A=@n5))) -> (Output_Get:bool)
   Do Once()
   Gate!(Enter:Exec, Open:Exec, Close:Exec, Toggle:Exec, bStartClosed:bool=true) -> (Exit:Exec) {is=title K2Node_MacroInstance}
   Outputs!()
  ]
 }
 ValuesLastTick {
  path=EventGraph.K2Node_Composite_2.ValuesLastTick
  nodes [
   Inputs()
   > Set IsUnderWaterLastTick(IsUnderWaterLastTick:bool=Get IsUnderWater()) -> (Output_Get:bool)
   Outputs()
  ]
 }
]
variables [
 AllowLocomotionBreak {type=boolean}
 BeginSkyDiveTime {type=number}
 CarrierAnimBP {type=M_MED_Carrier_Cosmos_Pack_AnimBP_C, default=null}
 Choose_LocoBreak_Index {type=integer}
 CowerSelection {type=integer}
 DelayMoveStart {type=boolean}
 EnteredWater {type=boolean}
 FortPlayerAnimBP {type=Fortnite_M_Avg_Player_AnimBlueprint_C, default=null}
 FortPlayerPawn {type=FortPlayerPawn}
 HasTargettedInLastNSeconds_b {type=boolean}
 IsCower {type=boolean}
 IsMovingOnGround {type=boolean}
 IsUnderWater {type=boolean}
 IsUnderWaterHunker {type=boolean}
 IsUnderWaterLastTick {type=boolean}
 JumpEnded {type=boolean}
 LocomotionBreaksNumber {type=integer, default=2}
 OwnerLocomotionDirection_b {type=number}
 OwnerLocomotionSpeed_b {type=number}
 PetInstCosmos {type=B_PetInst_Cosmos_C, default=null}
 SharpDirectionChange_b {type=boolean}
 StoppedMovingOrMeleeForLongEnoughTime_b {type=boolean}
 TookDamage {type=boolean}
 UnderWaterStopSelection {type=integer}
]
